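# Two-sided unlit translucent material for mesh-particle effects on the top of the Season 10 umbrella glider cosmetic.
# Particle color passes through a shared FX emissive adjustment function, and particle alpha drives opacity.
# Opacity is multiplied by vertical and horizontal UV edge masks and a powered normal-to-camera dot product, sign-corrected for backfaces to soften grazing edges.
# A depth fade softens intersections with scene geometry.
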Material /BRCosmetics/Gadgets/Assets/VinderTech_Umbrella/Umbrella_Season10/FX/M_S10_Umbrella_Top
graph {
 properties {
  bUsedWithMeshParticles=true
  bUsedWithNiagaraMeshParticles=true
  blendMode=BLEND_Translucent
  pixelDepthOffsetMode=PDOM_Legacy
  previewMesh=&/S_Umbrella_S10_Top
  shadingModel=MSM_Unlit
  twoSided=true
  usedShadingModels=Unlit
 }
 nodes [
  Constant3Vector() {constant=<rgba 0.012 0.5893 1 0>}
  n1 = ParticleColor()
  n2 = Multiply(
   A=MaterialFunctionCall!(UVs, Mult=ScalarParameter() {defaultValue=4, parameterName=Mult1}, Exp=ScalarParameter() {defaultValue=2, parameterName=Exp1}) {materialFunction=/Game/Effects/Fort_Effects/MaterialFunctions/MF_SquareEdgeMask}.Vertmask
   B=MaterialFunctionCall(Mult=ScalarParameter() {defaultValue=2.3, parameterName=Mult2}, Exp=ScalarParameter() {defaultValue=1.5, parameterName=Exp2}) {materialFunction=&MF_SquareEdgeMask}.HorizMask
  )
  n3 = DotProduct(A=Multiply(A=VertexNormalWS(), B=TwoSidedSign()), B=CameraVectorWS())
  n4 = ScalarParameter() {defaultValue=2, parameterName=EdgeExp}
  DepthFade(Opacity=Multiply(A=Multiply(A=@n1.A, B=@n2), B=Power(Base=@n3, Exp=@n4)), FadeDistance=ScalarParameter() {parameterName=FadeDist})
  MaterialFunctionCall(Emissive=@n1.RGB) {materialFunction=&MF_FX_EmissiveAdjustments}
 ]
}
