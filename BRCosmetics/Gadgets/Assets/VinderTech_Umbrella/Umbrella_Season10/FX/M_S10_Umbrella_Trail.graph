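# Unlit two-sided translucent material for the Season 10 umbrella trail, usable on beam trails and Niagara ribbons.
# Panning normal-map distortion warps two EdgedBand samples, one wrapped and one center-scaled and clamped.
# The clamped band blends two colors for emissive, while the maximum of both bands builds opacity.
# Edge masks, particle alpha and near-camera fade shape that opacity, with depth fade skipped on mobile.
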Material /BRCosmetics/Gadgets/Assets/VinderTech_Umbrella/Umbrella_Season10/FX/M_S10_Umbrella_Trail
graph {
 properties {
  bUsedWithBeamTrails=true
  bUsedWithNiagaraRibbons=true
  blendMode=BLEND_Translucent
  pixelDepthOffsetMode=PDOM_Legacy
  shadingModel=MSM_Unlit
  twoSided=true
  usedShadingModels=Unlit
 }
 nodes [
  n1 = VectorParameter() {parameterName=TilePan, defaultValue=<rgba 0.2 0.1 -0.3 0>}
  n2 = Panner(
   Coordinate=Multiply(A=TextureCoordinate() {coordinateIndex=1}, B=ComponentMask(None=@n1.RGB) {g=true, r=true})
   Speed=AppendVector(A=@n1.B, B=@n1.B)
  )
  n3 = TextureSampleParameter2D(UVs=Add(A=@n2, B=Constant2Vector() {r=0.25})) {parameterName=Distortion, samplerType=SAMPLERTYPE_Normal, texture=&T_Weap_Distort_N_02}
  n4 = ComponentMask(None=@n3.RGB) {g=true, r=true}
  n5 = ScalarParameter() {defaultValue=2.5, parameterName=DistAmntB}
  n6 = TextureCoordinate()
  n7 = Constant() {r=-0.5}
  n8 = Multiply(A=Add(A=@n6, B=@n7), B=ScalarParameter() {defaultValue=2, parameterName=Scale})
  n9 = Multiply(A=@n7) {constB=-1}
  n10 = TextureSampleParameter2D(UVs=Add(A=Multiply(A=@n4, B=@n5), B=Add(A=@n8, B=@n9))) {parameterName=Texture, samplerSource=SSM_Clamp_WorldGroupSettings, texture=/BRCosmetics/Athena/Effects/Textures/T_Particle_EdgedBand}
  n11 = Multiply(B=@n10.G) {constA=2}
  MaterialFunctionCall(Emissive=LinearInterpolate(
    A=VectorParameter() {parameterName=Color1, defaultValue=<rgba 0 0.6865 1 1>}.RGB
    B=VectorParameter() {parameterName=Color2, defaultValue=<rgba 0.7728 0.0827 1 1>}.RGB
    Alpha=Saturate(None=@n11)
   )) {materialFunction=&MF_FX_EmissiveAdjustments}
  n12 = TextureSampleParameter2D(UVs=@n2) {parameterName=Distortion, samplerType=SAMPLERTYPE_Normal, texture=&T_Weap_Distort_N_02}
  n13 = ComponentMask(None=@n12.RGB) {g=true, r=true}
  n14 = ScalarParameter() {defaultValue=1, parameterName=DistAmntA}
  n15 = TextureSampleParameter2D(UVs=Add(A=@n6, B=Multiply(A=@n13, B=@n14))) {parameterName=Texture, samplerSource=SSM_Wrap_WorldGroupSettings, texture=&T_Particle_EdgedBand}
  n16 = MaterialFunctionCall!(
   UVs
   Mult=ScalarParameter() {defaultValue=2, parameterName=RightMaskMult}
   Exp=ScalarParameter() {defaultValue=3, parameterName=RightMaskExp}
  ) {materialFunction=/Game/Effects/Fort_Effects/MaterialFunctions/MF_SquareEdgeMask}
  n17 = Multiply(
   A=MaterialFunctionCall(Mult=ScalarParameter() {defaultValue=20, parameterName=LeftMaskMult}, Exp=ScalarParameter() {defaultValue=1, parameterName=LeftMaskExp}) {materialFunction=&MF_SquareEdgeMask}.LeftMask
   B=MaterialFunctionCall(Mult=ScalarParameter() {defaultValue=20, parameterName=TopBotMaskMult}, Exp=ScalarParameter() {defaultValue=2, parameterName=TopBotMaskExp}) {materialFunction=&MF_SquareEdgeMask}.HorizMask
  )
  n18 = Multiply(A=Multiply(A=Max(A=@n15.G, B=@n10.G), B=Multiply(A=@n16.RightMask, B=@n17)), B=ParticleColor().A)
  n19 = ScalarParameter() {defaultValue=25, parameterName=FadeDist}
  Multiply(
   A=ShadingPathSwitch(Default=DepthFade(Opacity=@n18, FadeDistance=@n19), Mobile=@n18) {desc=Migrated from FeatureLevelSwitch}
   B=MaterialFunctionCall(Distance=ScalarParameter() {defaultValue=250, parameterName=NearFade}) {materialFunction=&Near_Camera_Fade}.Result
  )
  Saturate(None=Add(A=@n15.G, B=@n10.G))
 ]
}
