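# Niagara idle effect for the Season 10 VinderTech umbrella, showing blue and pink prop glows.
# A single burst emitter spawns particles that outlive their lifetime and follow a mesh socket through the system's socket reader.
# A curve on system age fades in the alpha of both colors after a brief delay.
# Local-space offsets from the socket position place the two glows, with two mesh and two sprite renderers drawing the effect.
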NiagaraSystem /BRCosmetics/Gadgets/Assets/VinderTech_Umbrella/Umbrella_Season10/FX/NS_Umbrella_S10_Idle
dependencies {
 usedDataInterfaces [&NiagaraDataInterfaceCurve, &NiagaraDataInterfaceSocketReader]
 usedDynamicInputs []
 usedModules [
  &self:SystemSpawnScript.SystemScriptSource.SystemScriptGraph.NiagaraNodeAssignment_0.SetVariables_#1
  &SystemState
 ]
}
emitters {Prop {
  emitterUpdateScript [
   &EmitterState(Life Cycle Mode=::NewEnumerator1, Loop Behavior=::NewEnumerator1, Loop Duration Mode=::NewEnumerator1, UseLoopDelay=false)
   &SpawnBurst_Instantaneous(Use Spawn Probability=false, UseLoopCountLimit=false)
  ]
  particleSpawnScript [
   &InitializeParticle(Color Mode=::NewEnumerator0, Lifetime=5, Position Mode=::NewEnumerator0, Write Material Random=true)
   /SetVariables(Particles.GlowColorAlignment:Vector3f=<0 0 1>) {isSetParameters=true}
  ]
  particleUpdateScript [
   &ParticleState(Kill Particles When Lifetime Has Elapsed=false, Let Infinitely Lived Particles Die When Emitter Deactivates=true)
   &SocketLocation(
    SamplePosOffset=true
    Sampled Position Offset=<0 0 3>
    Sampled Position Offset Coordinate Space:ENiagaraSkelMeshTransforms=::NewEnumerator4
    Socket Reader=$System.Socket Reader
    Socket Sampling Mode=::NewEnumerator3
    VelocitySpeedLimit=false
   )
   /SetVariables(Particles.StartupFade:NiagaraFloat=&FloatFromCurve(
     FloatCurve=DataInterface() {propertyValues {Curve {defaultValue=0, keys [{tangentMode=RCTM_User, time=0.25, value=0}, {tangentMode=RCTM_User, time=0.75, value=1}]}}}
     CurveIndex=$Engine.System.Age
    )) {isSetParameters=true}
   /SetVariables(
    Particles.BlueGlowColor:LinearColor=&MakeLinearColorFromVectorAndFloat(
     "Vector (RGB)"=&MakeVectorFromLinearColorRGB(Color=<rgba 0.0999 0.5972 1 1>)
     "Float (Alpha)"=&Multiply_Float(A=0.12, B=$Particles.StartupFade)
    )
    Particles.BluePropColor:LinearColor=&MakeLinearColorFromVectorAndFloat(
     "Vector (RGB)"=&MakeVectorFromLinearColorRGB(Color=<rgba 0.0999 0.5972 1 1>)
     "Float (Alpha)"=&Multiply_Float(A=0.12, B=$Particles.StartupFade)
    )
    Particles.PinkGlowCOlor:LinearColor=&MakeLinearColorFromVectorAndFloat(
     "Vector (RGB)"=&MakeVectorFromLinearColorRGB(Color=<rgba 1 0.2016 1 1>)
     "Float (Alpha)"=&Multiply_Float(A=0.1, B=$Particles.StartupFade)
    )
    Particles.PinkPropColor:LinearColor=&MakeLinearColorFromVectorAndFloat(
     "Vector (RGB)"=&MakeVectorFromLinearColorRGB(Color=<rgba 1 0.2016 1 1>)
     "Float (Alpha)"=&Multiply_Float(A=0.1, B=$Particles.StartupFade)
    )
   ) {isSetParameters=true}
   /SetVariables(
    Particles.GlowPositionBlue:NiagaraPosition=&AddVectorToPosition(
     Position=$Output.SocketLocation.SampledPosition
     Vector=&ConvertVectorToPosition(
      Input Position=&TransformPosition(Position=<0 0 5>, Source Space=::Local, Destination Space=::Simulation)
      Conversion Operation:ENiagara_LWCConvertVecToPos=::NewEnumerator0
     )
    )
    Particles.GlowPositionPink:NiagaraPosition=&AddVectorToPosition(
     Position=$Output.SocketLocation.SampledPosition
     Vector=&ConvertVectorToPosition(
      Input Position=&TransformPosition(Position=<0 0 3>, Source Space=::Local, Destination Space=::Simulation)
      Conversion Operation:ENiagara_LWCConvertVecToPos=::NewEnumerator0
     )
    )
   ) {isSetParameters=true}
  ]
  rendererClasses [&NiagaraMeshRendererProperties, &NiagaraSpriteRendererProperties]
  renderers [
   {rendererClass=&NiagaraMeshRendererProperties, rendererIndex=0}
   {rendererClass=&NiagaraMeshRendererProperties, rendererIndex=1}
   {rendererClass=&NiagaraSpriteRendererProperties, rendererIndex=2}
   {rendererClass=&NiagaraSpriteRendererProperties, rendererIndex=3}
  ]
 }}
summary {
 systemName=NS_Umbrella_S10_Idle
 emitters [{emitterName=Prop, rendererClasses [&NiagaraMeshRendererProperties, &NiagaraSpriteRendererProperties]}]
 userVariables []
}
